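# Sound cue for light debris falling from the cave ceiling, a randomized ambient spot sound in the Hollow Base POI.
# A mixer layers one random rock debris wave with one random pebble fall borrowed from Cattus Cave, weighting the pebble layer louder.
# Each layer passes through its own modulator.
# A less-positional one-shot attenuation applies, and overridden concurrency blocks a new instance while one plays on the same owner.
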SoundCue /Game/Sounds/Artemis/POI/Hollow_Base/Spots/Spot_Cave_Ceiling_Debris_Fall_Light_Random_Cue
root=Mixer_0
nodes [
 {id=Mixer_0, children [Modulator_2,Modulator_4], properties {inputVolume [1,2]}}
 {id=Modulator_2, children [Random_0]}
 {id=Modulator_4, children [Random_3]}
 {
  id=Random_0
  children [WavePlayer_0, WavePlayer_1, WavePlayer_2, WavePlayer_3, WavePlayer_4, WavePlayer_11, WavePlayer_15, WavePlayer_16, WavePlayer_17]
  properties {weights [1,1,1,1,1,1,1,1,1]}
 }
 {
  id=Random_3
  children [WavePlayer_27, WavePlayer_28, WavePlayer_29, WavePlayer_30, WavePlayer_31, WavePlayer_32]
  properties {weights [1,1,1,1,1,1]}
 }
 {id=WavePlayer_0, properties {soundWaveAssetPtr=&/Spot_Cave_Rock_Debris_Falling_01}}
 {id=WavePlayer_1, properties {soundWaveAssetPtr=&/Spot_Cave_Rock_Debris_Falling_02}}
 {id=WavePlayer_11, properties {soundWaveAssetPtr=&/Spot_Cave_Rock_Debris_Falling_06}}
 {id=WavePlayer_15, properties {soundWaveAssetPtr=&/Spot_Cave_Rock_Debris_Falling_07}}
 {id=WavePlayer_16, properties {soundWaveAssetPtr=&/Spot_Cave_Rock_Debris_Falling_08}}
 {id=WavePlayer_17, properties {soundWaveAssetPtr=&/Spot_Cave_Rock_Debris_Falling_09}}
 {id=WavePlayer_2, properties {soundWaveAssetPtr=&/Spot_Cave_Rock_Debris_Falling_03}}
 {id=WavePlayer_27, properties {soundWaveAssetPtr=/Game/Sounds/Artemis/POI/Cattus_Cave/Spots/Spot_Cattus_Cave_Pebble_Fall_01}}
 {id=WavePlayer_28, properties {soundWaveAssetPtr=/Game/Sounds/Artemis/POI/Cattus_Cave/Spots/Spot_Cattus_Cave_Pebble_Fall_02}}
 {id=WavePlayer_29, properties {soundWaveAssetPtr=/Game/Sounds/Artemis/POI/Cattus_Cave/Spots/Spot_Cattus_Cave_Pebble_Fall_03}}
 {id=WavePlayer_3, properties {soundWaveAssetPtr=&/Spot_Cave_Rock_Debris_Falling_04}}
 {id=WavePlayer_30, properties {soundWaveAssetPtr=/Game/Sounds/Artemis/POI/Cattus_Cave/Spots/Spot_Cattus_Cave_Pebble_Fall_04}}
 {id=WavePlayer_31, properties {soundWaveAssetPtr=/Game/Sounds/Artemis/POI/Cattus_Cave/Spots/Spot_Cattus_Cave_Pebble_Fall_05}}
 {id=WavePlayer_32, properties {soundWaveAssetPtr=/Game/Sounds/Artemis/POI/Cattus_Cave/Spots/Spot_Cattus_Cave_Pebble_Fall_06}}
 {id=WavePlayer_4, properties {soundWaveAssetPtr=&/Spot_Cave_Rock_Debris_Falling_05}}
]
properties {
 attenuationSettings=/Game/Sounds/Attenuation/Fort_TileAmbient_OneShot_LessPositional_Log_Att
 bEnableBusSends=false
 bEnableSubmixSends=false
 bOverrideConcurrency=true
 maxDistance=6200
 soundClassObject=&Fort_Ambient_Outdoors_Athena
 volumeMultiplier=0.5
 attenuationOverrides {bAttenuate=false}
 concurrencyOverrides {bLimitToOwner=true, maxCount=1, resolutionRule=PreventNew}
}
