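# Sound cue for a distant, light pebble-fall spot ambience in the Hollow Base cave POI.
# A Random node picks among six equally weighted pebble-fall waves reused from Cattus Cave, then passes the pick through a Modulator for variation.
# It plays in the outdoor ambient sound class with a less-positional one-shot attenuation asset, and has bus and submix sends disabled.
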SoundCue /Game/Sounds/Artemis/POI/Hollow_Base/Spots/Spot_Cave_Distant_Light_Pebble_Fall_Cue
root=Modulator_0
nodes [
 {id=Modulator_0, children [Random_1]}
 {
  id=Random_1
  children [WavePlayer_5, WavePlayer_6, WavePlayer_7, WavePlayer_8, WavePlayer_9, WavePlayer_10]
  properties {weights [1,1,1,1,1,1]}
 }
 {id=WavePlayer_10, properties {soundWaveAssetPtr=/Game/Sounds/Artemis/POI/Cattus_Cave/Spots/Spot_Cattus_Cave_Pebble_Fall_06}}
 {id=WavePlayer_5, properties {soundWaveAssetPtr=/Game/Sounds/Artemis/POI/Cattus_Cave/Spots/Spot_Cattus_Cave_Pebble_Fall_01}}
 {id=WavePlayer_6, properties {soundWaveAssetPtr=/Game/Sounds/Artemis/POI/Cattus_Cave/Spots/Spot_Cattus_Cave_Pebble_Fall_02}}
 {id=WavePlayer_7, properties {soundWaveAssetPtr=/Game/Sounds/Artemis/POI/Cattus_Cave/Spots/Spot_Cattus_Cave_Pebble_Fall_03}}
 {id=WavePlayer_8, properties {soundWaveAssetPtr=/Game/Sounds/Artemis/POI/Cattus_Cave/Spots/Spot_Cattus_Cave_Pebble_Fall_04}}
 {id=WavePlayer_9, properties {soundWaveAssetPtr=/Game/Sounds/Artemis/POI/Cattus_Cave/Spots/Spot_Cattus_Cave_Pebble_Fall_05}}
]
properties {
 attenuationSettings=/Game/Sounds/Attenuation/Fort_TileAmbient_OneShot_LessPositional_Log_Att
 bEnableBusSends=false
 bEnableSubmixSends=false
 maxDistance=6200
 soundClassObject=&Fort_Ambient_Outdoors_Athena
 volumeMultiplier=1.002
 attenuationOverrides {bAttenuate=false}
}
