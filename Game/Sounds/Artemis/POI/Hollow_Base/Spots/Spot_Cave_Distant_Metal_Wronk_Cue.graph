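# Sound cue for a distant metal "wronk" ambient one-shot, a cave spot sound for the Hollow Base point of interest.
# A random node picks evenly among five wave variations, then a modulator randomizes pitch and volume on each play.
# It plays quietly in the indoor ambient one-shot sound class, using positional tile-ambient attenuation, with bus and submix sends disabled.
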SoundCue /Game/Sounds/Artemis/POI/Hollow_Base/Spots/Spot_Cave_Distant_Metal_Wronk_Cue
root=Modulator_0
nodes [
 {id=Modulator_0, children [Random_0], properties {pitchMin=0.85, volumeMin=0.8}}
 {id=Random_0, children [WavePlayer_0,WavePlayer_1,WavePlayer_2,WavePlayer_4,WavePlayer_7], properties {weights [1,1,1,1,1]}}
 {id=WavePlayer_0, properties {soundWaveAssetPtr=&/Spot_Cave_Distant_Metal_Wronk_1}}
 {id=WavePlayer_1, properties {soundWaveAssetPtr=&/Spot_Cave_Distant_Metal_Wronk_2}}
 {id=WavePlayer_2, properties {soundWaveAssetPtr=&/Spot_Cave_Distant_Metal_Wronk_3}}
 {id=WavePlayer_4, properties {soundWaveAssetPtr=&/Spot_Cave_Distant_Metal_Wronk_5}}
 {id=WavePlayer_7, properties {soundWaveAssetPtr=&/Spot_Cave_Distant_Metal_Wronk_8}}
]
properties {
 attenuationSettings=/Game/Sounds/Attenuation/Fort_TileAmbient_OneShot_MorePositional_Att
 bEnableBusSends=false
 bEnableSubmixSends=false
 maxDistance=8000
 soundClassObject=/Game/Sounds/FortSoundClassesAndModes/Fort_Ambient_OneShots_Indoors
 volumeMultiplier=0.251
 attenuationOverrides {falloffDistance=6000, stereoSpread=10000, attenuationShapeExtents {x=2000}}
}
