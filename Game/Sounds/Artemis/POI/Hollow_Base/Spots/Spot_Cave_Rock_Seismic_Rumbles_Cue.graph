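# Ambient spot cue for Hollow Base cave seismic rumbles with rattling rock.
# The first two quality levels mix pitch-varied rumbles with a separately attenuated rock layer, randomly choosing single rattles, enveloped staggered rattle pairs, or pairs plus pitched-down slate debris.
# The last two levels use a simpler mix of delayed rattles and two rumble variants, without attenuation nodes.
# Random delays stagger the rock layers behind the rumbles.
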SoundCue /Game/Sounds/Artemis/POI/Hollow_Base/Spots/Spot_Cave_Rock_Seismic_Rumbles_Cue
root=QualityLevel_0
nodes [
 {id=Attenuation_0, children [Modulator_1], properties {attenuationSettings=/Game/Sounds/Attenuation/Fort_TileAmbient_OneShot_Att_Closer}}
 {id=Attenuation_1, children [Random_4], properties {attenuationSettings=/Game/Sounds/Attenuation/Fort_TileAmbient_OneShot_Distant_Creak_Att}}
 {id=Delay_0, children [Random_0], properties {delayMax=3, delayMin=2}}
 {id=Delay_1, children [Random_3], properties {delayMax=4, delayMin=3}}
 {id=Delay_2, children [Random_6], properties {delayMax=3, delayMin=2}}
 {id=Delay_3, children [Random_8], properties {delayMax=3, delayMin=2}}
 {id=Delay_4, children [Modulator_2], properties {delayMax=4, delayMin=2}}
 {
  id=Enveloper_0
  children [Delay_0]
  properties {
   pitchCurve {editorCurveData {keys [{time=0}]}}
   volumeCurve {editorCurveData {keys [
      {arriveTangent=10.3243, interpMode=RCIM_Constant, time=0, value=1}
      {arriveTangent=-0.0065, interpMode=RCIM_Cubic, leaveTangent=-0.0065, tangentMode=RCTM_User, time=4, value=1}
      {arriveTangent=-5.777, interpMode=RCIM_Cubic, leaveTangent=-9.0114, tangentMode=RCTM_User, time=4.3, value=0}
     ]}}
  }
 }
 {
  id=Enveloper_1
  children [Delay_1]
  properties {
   pitchCurve {editorCurveData {keys [{time=0}]}}
   volumeCurve {editorCurveData {keys [{arriveTangent=10.3243, interpMode=RCIM_Cubic, leaveTangent=10.3243, tangentMode=RCTM_User, time=0, value=0}, <at 0.3 1 RCIM_Cubic>]}}
  }
 }
 {
  id=Enveloper_2
  children [Delay_4]
  properties {
   pitchCurve {editorCurveData {keys [{time=0}]}}
   volumeCurve {editorCurveData {keys [{arriveTangent=3.1397, interpMode=RCIM_Cubic, leaveTangent=3.1397, tangentMode=RCTM_User, time=0, value=0}, <at 1 1 RCIM_Cubic>]}}
  }
 }
 {id=Mixer_0, children [Attenuation_1, Attenuation_0], properties {inputVolume [1,2]}}
 {id=Mixer_1, children [Enveloper_0,Enveloper_1], properties {inputVolume [0.5,0.5]}}
 {id=Mixer_2, children [Mixer_1,Enveloper_2], properties {inputVolume [1,0.3]}}
 {id=Mixer_3, children [Delay_3,Random_7], properties {inputVolume [1,1]}}
 {id=Mixer_5, children [Delay_2], properties {inputVolume [0.4]}}
 {id=Modulator_0, children [Mixer_0], properties {volumeMin=1}}
 {id=Modulator_1, children [Random_2], properties {pitchMax=1.7, pitchMin=0.8, volumeMax=1, volumeMin=0.8}}
 {id=Modulator_2, children [Random_5], properties {pitchMax=0.85, pitchMin=0.8}}
 {id=Modulator_3, children [Mixer_3], properties {volumeMin=1}}
 {id=QualityLevel_0, children [Modulator_0,Modulator_0,Modulator_3,Modulator_3]}
 {id=Random_0, children [WavePlayer_0,WavePlayer_1,WavePlayer_2,WavePlayer_3,WavePlayer_4], properties {weights [1,1,1,1,1]}}
 {
  id=Random_2
  children [WavePlayer_20, WavePlayer_21, WavePlayer_22, WavePlayer_23, WavePlayer_24, WavePlayer_25]
  properties {weights [1,1,1,1,1,1]}
 }
 {id=Random_3, children [WavePlayer_26, WavePlayer_27, WavePlayer_28, WavePlayer_29, WavePlayer_30], properties {weights [1,1,1,1,1]}}
 {id=Random_4, children [Mixer_5,Mixer_1,Mixer_2], properties {weights [1,1,1]}}
 {
  id=Random_5
  children [WavePlayer_31, WavePlayer_32, WavePlayer_33, WavePlayer_34, WavePlayer_35, WavePlayer_36, WavePlayer_37, WavePlayer_38]
  properties {weights [1,1,1,1,1,1,1,1]}
 }
 {
  id=Random_6
  children [
   WavePlayer_39
   WavePlayer_40
   WavePlayer_41
   WavePlayer_42
   WavePlayer_43
   WavePlayer_44
   WavePlayer_45
   WavePlayer_46
   WavePlayer_47
   WavePlayer_48
  ]
  properties {weights [1,1,1,1,1,1,1,1,1,1]}
 }
 {id=Random_7, children [WavePlayer_49, WavePlayer_50], properties {weights [1,1]}}
 {id=Random_8, children [WavePlayer_51, WavePlayer_52, WavePlayer_53, WavePlayer_54, WavePlayer_55], properties {weights [1,1,1,1,1]}}
 {id=WavePlayer_0, properties {soundWaveAssetPtr=&/Spot_Cave_Rock_Rattle_01}}
 {id=WavePlayer_1, properties {soundWaveAssetPtr=&/Spot_Cave_Rock_Rattle_02}}
 {id=WavePlayer_2, properties {soundWaveAssetPtr=&/Spot_Cave_Rock_Rattle_03}}
 {id=WavePlayer_20, properties {soundWaveAssetPtr=&/Spot_Cave_Seismic_Rumble_01}}
 {id=WavePlayer_21, properties {soundWaveAssetPtr=&/Spot_Cave_Seismic_Rumble_02}}
 {id=WavePlayer_22, properties {soundWaveAssetPtr=&/Spot_Cave_Seismic_Rumble_03}}
 {id=WavePlayer_23, properties {soundWaveAssetPtr=&/Spot_Cave_Seismic_Rumble_04}}
 {id=WavePlayer_24, properties {soundWaveAssetPtr=&/Spot_Cave_Seismic_Rumble_05}}
 {id=WavePlayer_25, properties {soundWaveAssetPtr=&/Spot_Cave_Seismic_Rumble_06}}
 {id=WavePlayer_26, properties {soundWaveAssetPtr=&/Spot_Cave_Rock_Rattle_06}}
 {id=WavePlayer_27, properties {soundWaveAssetPtr=&/Spot_Cave_Rock_Rattle_07}}
 {id=WavePlayer_28, properties {soundWaveAssetPtr=&/Spot_Cave_Rock_Rattle_08}}
 {id=WavePlayer_29, properties {soundWaveAssetPtr=&/Spot_Cave_Rock_Rattle_10}}
 {id=WavePlayer_3, properties {soundWaveAssetPtr=&/Spot_Cave_Rock_Rattle_04}}
 {id=WavePlayer_30, properties {soundWaveAssetPtr=&/Spot_Cave_Rock_Rattle_12}}
 {id=WavePlayer_31, properties {soundWaveAssetPtr=&/Spot_Cave_Rock_Slate_Debris_04}}
 {id=WavePlayer_32, properties {soundWaveAssetPtr=&/Spot_Cave_Rock_Slate_Debris_06}}
 {id=WavePlayer_33, properties {soundWaveAssetPtr=&/Spot_Cave_Rock_Slate_Debris_08}}
 {id=WavePlayer_34, properties {soundWaveAssetPtr=&/Spot_Cave_Rock_Slate_Debris_10}}
 {id=WavePlayer_35, properties {soundWaveAssetPtr=&/Spot_Cave_Rock_Slate_Debris_05}}
 {id=WavePlayer_36, properties {soundWaveAssetPtr=&/Spot_Cave_Rock_Slate_Debris_07}}
 {id=WavePlayer_37, properties {soundWaveAssetPtr=&/Spot_Cave_Rock_Slate_Debris_11}}
 {id=WavePlayer_38, properties {soundWaveAssetPtr=&/Spot_Cave_Rock_Slate_Debris_12}}
 {id=WavePlayer_39, properties {soundWaveAssetPtr=&Spot_Cave_Rock_Rattle_01}}
 {id=WavePlayer_4, properties {soundWaveAssetPtr=&/Spot_Cave_Rock_Rattle_05}}
 {id=WavePlayer_40, properties {soundWaveAssetPtr=&Spot_Cave_Rock_Rattle_02}}
 {id=WavePlayer_41, properties {soundWaveAssetPtr=&Spot_Cave_Rock_Rattle_03}}
 {id=WavePlayer_42, properties {soundWaveAssetPtr=&Spot_Cave_Rock_Rattle_04}}
 {id=WavePlayer_43, properties {soundWaveAssetPtr=&Spot_Cave_Rock_Rattle_05}}
 {id=WavePlayer_44, properties {soundWaveAssetPtr=&Spot_Cave_Rock_Rattle_06}}
 {id=WavePlayer_45, properties {soundWaveAssetPtr=&Spot_Cave_Rock_Rattle_07}}
 {id=WavePlayer_46, properties {soundWaveAssetPtr=&Spot_Cave_Rock_Rattle_08}}
 {id=WavePlayer_47, properties {soundWaveAssetPtr=&Spot_Cave_Rock_Rattle_10}}
 {id=WavePlayer_48, properties {soundWaveAssetPtr=&Spot_Cave_Rock_Rattle_12}}
 {id=WavePlayer_49, properties {soundWaveAssetPtr=&Spot_Cave_Seismic_Rumble_01}}
 {id=WavePlayer_50, properties {soundWaveAssetPtr=&Spot_Cave_Seismic_Rumble_02}}
 {id=WavePlayer_51, properties {soundWaveAssetPtr=&Spot_Cave_Rock_Rattle_01}}
 {id=WavePlayer_52, properties {soundWaveAssetPtr=&Spot_Cave_Rock_Rattle_02}}
 {id=WavePlayer_53, properties {soundWaveAssetPtr=&Spot_Cave_Rock_Rattle_03}}
 {id=WavePlayer_54, properties {soundWaveAssetPtr=&Spot_Cave_Rock_Rattle_04}}
 {id=WavePlayer_55, properties {soundWaveAssetPtr=&Spot_Cave_Rock_Rattle_05}}
]
properties {
 bEnableBusSends=false
 bEnableSubmixSends=false
 maxDistance=8500
 soundClassObject=&Fort_Ambient_Outdoors_Athena
 volumeMultiplier=0.501
 attenuationOverrides {bAttenuate=false}
}
