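# Sound cue for spot ambience of seismic rock rumbles in the Hollow Base cave interior, on the outdoor ambient sound class.
# A quality-level node sends its first two slots to six random rumble waves with wide pitch and slight volume variation.
# Its last two slots use a lighter branch of only two waves without that wide variation.
# Both branches boost level through a mixer and share close-range one-shot attenuation.
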SoundCue /Game/Sounds/Artemis/POI/Hollow_Base/Spots/Spot_Cave_Rock_Seismic_Rumbles_Interior_Cue
root=QualityLevel_0
nodes [
 {id=Attenuation_0, children [Modulator_1], properties {attenuationSettings=/Game/Sounds/Attenuation/Fort_TileAmbient_OneShot_Att_Closer}}
 {id=Attenuation_2, children [Modulator_3], properties {attenuationSettings=&Fort_TileAmbient_OneShot_Att_Closer}}
 {id=Mixer_0, children [Random_2], properties {inputVolume [2]}}
 {id=Mixer_3, children [Random_7], properties {inputVolume [3]}}
 {id=Modulator_0, children [Attenuation_0], properties {volumeMin=1}}
 {id=Modulator_1, children [Mixer_0], properties {pitchMax=1.7, pitchMin=0.8, volumeMax=1, volumeMin=0.8}}
 {id=Modulator_3, children [Mixer_3], properties {volumeMin=1}}
 {id=QualityLevel_0, children [Modulator_0, Modulator_0, Attenuation_2, Attenuation_2]}
 {
  id=Random_2
  children [WavePlayer_20, WavePlayer_21, WavePlayer_22, WavePlayer_23, WavePlayer_24, WavePlayer_25]
  properties {weights [1,1,1,1,1,1]}
 }
 {id=Random_7, children [WavePlayer_49, WavePlayer_50], properties {weights [1,1]}}
 {id=WavePlayer_20, properties {soundWaveAssetPtr=&/Spot_Cave_Seismic_Rumble_01}}
 {id=WavePlayer_21, properties {soundWaveAssetPtr=&/Spot_Cave_Seismic_Rumble_02}}
 {id=WavePlayer_22, properties {soundWaveAssetPtr=&/Spot_Cave_Seismic_Rumble_03}}
 {id=WavePlayer_23, properties {soundWaveAssetPtr=&/Spot_Cave_Seismic_Rumble_04}}
 {id=WavePlayer_24, properties {soundWaveAssetPtr=&/Spot_Cave_Seismic_Rumble_05}}
 {id=WavePlayer_25, properties {soundWaveAssetPtr=&/Spot_Cave_Seismic_Rumble_06}}
 {id=WavePlayer_49, properties {soundWaveAssetPtr=&Spot_Cave_Seismic_Rumble_01}}
 {id=WavePlayer_50, properties {soundWaveAssetPtr=&Spot_Cave_Seismic_Rumble_02}}
]
properties {
 bEnableBusSends=false
 bEnableSubmixSends=false
 maxDistance=8500
 soundClassObject=&Fort_Ambient_Outdoors_Athena
 volumeMultiplier=0.501
 attenuationOverrides {bAttenuate=false}
}
