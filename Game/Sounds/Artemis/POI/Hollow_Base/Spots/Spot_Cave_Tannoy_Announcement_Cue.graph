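# Sound Cue for a cave tannoy announcement ambient spot sound at the Hollow Base POI.
# A Random node picks one of thirteen equally weighted announcement waves and passes it through a Modulator at default settings.
# It uses the shared tile-ambient one-shot attenuation and the outdoor ambient sound class.
# Submix sends are off, while a distance-based bus send feeds the ambient reflection source bus.
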SoundCue /Game/Sounds/Artemis/POI/Hollow_Base/Spots/Spot_Cave_Tannoy_Announcement_Cue
root=Modulator_0
nodes [
 {id=Modulator_0, children [Random_0]}
 {
  id=Random_0
  children [
   WavePlayer_1
   WavePlayer_2
   WavePlayer_3
   WavePlayer_4
   WavePlayer_5
   WavePlayer_6
   WavePlayer_7
   WavePlayer_8
   WavePlayer_0
   WavePlayer_9
   WavePlayer_10
   WavePlayer_11
   WavePlayer_12
  ]
  properties {weights [1,1,1,1,1,1,1,1,1,1,1,1,1]}
 }
 {id=WavePlayer_0, properties {soundWaveAssetPtr=&/Spot_Cave_Tannoy_Announcement_09}}
 {id=WavePlayer_1, properties {soundWaveAssetPtr=&/Spot_Cave_Tannoy_Announcement_01}}
 {id=WavePlayer_10, properties {soundWaveAssetPtr=&/Spot_Cave_Tannoy_Announcement_11}}
 {id=WavePlayer_11, properties {soundWaveAssetPtr=&/Spot_Cave_Tannoy_Announcement_12}}
 {id=WavePlayer_12, properties {soundWaveAssetPtr=&/Spot_Cave_Tannoy_Announcement_13}}
 {id=WavePlayer_2, properties {soundWaveAssetPtr=&/Spot_Cave_Tannoy_Announcement_02}}
 {id=WavePlayer_3, properties {soundWaveAssetPtr=&/Spot_Cave_Tannoy_Announcement_03}}
 {id=WavePlayer_4, properties {soundWaveAssetPtr=&/Spot_Cave_Tannoy_Announcement_04}}
 {id=WavePlayer_5, properties {soundWaveAssetPtr=&/Spot_Cave_Tannoy_Announcement_05}}
 {id=WavePlayer_6, properties {soundWaveAssetPtr=&/Spot_Cave_Tannoy_Announcement_06}}
 {id=WavePlayer_7, properties {soundWaveAssetPtr=&/Spot_Cave_Tannoy_Announcement_07}}
 {id=WavePlayer_8, properties {soundWaveAssetPtr=&/Spot_Cave_Tannoy_Announcement_08}}
 {id=WavePlayer_9, properties {soundWaveAssetPtr=&/Spot_Cave_Tannoy_Announcement_10}}
]
properties {
 attenuationSettings=/Game/Sounds/Attenuation/Fort_TileAmbient_OneShot_Att
 bEnableSubmixSends=false
 maxDistance=8000
 soundClassObject=&Fort_Ambient_Outdoors_Athena
 volumeMultiplier=1.002
 attenuationOverrides {bAttenuate=false}
 busSends [{
   audioBus=null
   maxSendDistance=3500
   minSendDistance=3000
   minSendLevel=0
   sendLevel=1
   soundSourceBus=/Game/Sounds/Apollo/Biomes/Reflections/AmbientAudio_Reflection_SourceBus
  }]
}
