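# Looping ambient spot cue for a sizzling drill bit at the Hollow Base POI.
# A looping wave player feeds an Enveloper root whose volume curve ramps from silent to full over the first second, fading the loop in.
# The cue plays at reduced volume through the outdoor ambient sound class, with a dedicated attenuation asset shaping its spatial falloff.
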SoundCue /Game/Sounds/Artemis/POI/Hollow_Base/Spots/Spot_Drill_Bit_Sizzle_Loop_Cue
root=Enveloper_0
nodes [
 {
  id=Enveloper_0
  children [WavePlayer_0]
  properties {pitchCurve {editorCurveData {keys [{time=0}]}}, volumeCurve {editorCurveData {keys [<at 0 0>,<at 1 1>]}}}
 }
 {id=WavePlayer_0, properties {bLooping=true, soundWaveAssetPtr=&/Spot_Drill_Bit_Sizzle_Loop}}
]
properties {
 attenuationSettings=&/Spot_Drill_Bit_Sizzle_Att
 maxDistance=3000
 soundClassObject=&Fort_Ambient_Outdoors_Athena
 volumeMultiplier=0.501
}
