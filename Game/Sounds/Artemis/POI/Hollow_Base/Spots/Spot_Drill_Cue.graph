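# Sound cue for a spot drill ambient sound at the Hollow Base POI, assigned to an indoor ambient sound class.
# A random node picks one of six equally weighted drill waves and feeds a modulator that slightly varies pitch and volume, atop a lowered base pitch and boosted volume.
# It references a dedicated attenuation asset, and its concurrency override allows one instance at a time, refusing new ones.
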SoundCue /Game/Sounds/Artemis/POI/Hollow_Base/Spots/Spot_Drill_Cue
root=Modulator_0
nodes [
 {id=Modulator_0, children [Random_0], properties {pitchMax=1, pitchMin=0.9, volumeMax=1, volumeMin=0.8}}
 {
  id=Random_0
  children [WavePlayer_3, WavePlayer_4, WavePlayer_5, WavePlayer_6, WavePlayer_7, WavePlayer_11]
  properties {bShouldExcludeFromBranchCulling=true, weights [1,1,1,1,1,1]}
 }
 {id=WavePlayer_11, properties {soundWaveAssetPtr=&/Spot_Drill_06}}
 {id=WavePlayer_3, properties {soundWaveAssetPtr=&/Spot_Drill_01}}
 {id=WavePlayer_4, properties {soundWaveAssetPtr=&/Spot_Drill_02}}
 {id=WavePlayer_5, properties {soundWaveAssetPtr=&/Spot_Drill_03}}
 {id=WavePlayer_6, properties {soundWaveAssetPtr=&/Spot_Drill_04}}
 {id=WavePlayer_7, properties {soundWaveAssetPtr=&/Spot_Drill_05}}
]
properties {
 attenuationSettings=&/Spot_Drill_Att
 bEnableBusSends=false
 bEnableSubmixSends=false
 bOverrideConcurrency=true
 maxDistance=34000
 pitchMultiplier=0.7
 soundClassObject=&Fort_Ambient_Indoors_Athena
 volumeMultiplier=1.504
 attenuationOverrides {falloffDistance=6000, stereoSpread=10000, attenuationShapeExtents {x=2000}}
 concurrencyOverrides {maxCount=1, resolutionRule=PreventNew}
}
