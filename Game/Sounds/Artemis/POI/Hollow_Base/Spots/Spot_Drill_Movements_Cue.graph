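# Sound cue for drill earth-movement ambient spot sounds at the Hollow Base POI, in the outdoor ambient sound class.
# A quality-level root feeds equal-weight random pickers: the first two levels share eight movement variations, while the last two share only variations 06 to 08, trimming variety by quality setting.
# A drill-movement attenuation asset handles spatial falloff, volume is reduced, and bus and submix sends are disabled.
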SoundCue /Game/Sounds/Artemis/POI/Hollow_Base/Spots/Spot_Drill_Movements_Cue
root=QualityLevel_0
nodes [
 {id=QualityLevel_0, children [Random_1,Random_1,Random_9,Random_9]}
 {
  id=Random_1
  children [WavePlayer_5, WavePlayer_6, WavePlayer_7, WavePlayer_8, WavePlayer_9, WavePlayer_10, WavePlayer_11, WavePlayer_12]
  properties {weights [1,1,1,1,1,1,1,1]}
 }
 {id=Random_9, children [WavePlayer_13, WavePlayer_14, WavePlayer_15], properties {weights [1,1,1]}}
 {id=WavePlayer_10, properties {soundWaveAssetPtr=&/Spot_Drill_Earth_Movement_06}}
 {id=WavePlayer_11, properties {soundWaveAssetPtr=&/Spot_Drill_Earth_Movement_07}}
 {id=WavePlayer_12, properties {soundWaveAssetPtr=&/Spot_Drill_Earth_Movement_08}}
 {id=WavePlayer_13, properties {soundWaveAssetPtr=&Spot_Drill_Earth_Movement_06}}
 {id=WavePlayer_14, properties {soundWaveAssetPtr=&Spot_Drill_Earth_Movement_07}}
 {id=WavePlayer_15, properties {soundWaveAssetPtr=&Spot_Drill_Earth_Movement_08}}
 {id=WavePlayer_5, properties {soundWaveAssetPtr=&/Spot_Drill_Earth_Movement_01}}
 {id=WavePlayer_6, properties {soundWaveAssetPtr=&/Spot_Drill_Earth_Movement_02}}
 {id=WavePlayer_7, properties {soundWaveAssetPtr=&/Spot_Drill_Earth_Movement_03}}
 {id=WavePlayer_8, properties {soundWaveAssetPtr=&/Spot_Drill_Earth_Movement_04}}
 {id=WavePlayer_9, properties {soundWaveAssetPtr=&/Spot_Drill_Earth_Movement_05}}
]
properties {
 attenuationSettings=&/Spot_Drill_Movement_Att
 bEnableBusSends=false
 bEnableSubmixSends=false
 maxDistance=10000
 soundClassObject=&Fort_Ambient_Outdoors_Athena
 volumeMultiplier=0.652
 attenuationOverrides {bAttenuate=false}
}
