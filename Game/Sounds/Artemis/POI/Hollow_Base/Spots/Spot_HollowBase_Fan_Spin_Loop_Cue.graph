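# Looping ambient spot sound cue for a spinning fan at the Hollow Base POI.
# Separate blade and airy loops each pass through their own attenuation into a Mixer that weights the air layer higher.
# An Enveloper at the root ramps the combined volume up from silence as a fade-in.
# The cue plays in the ambient props sound class and overrides concurrency to cap simultaneous instances.
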SoundCue /Game/Sounds/Artemis/POI/Hollow_Base/Spots/Spot_HollowBase_Fan_Spin_Loop_Cue
root=Enveloper_0
nodes [
 {
  id=Attenuation_0
  children [WavePlayer_4]
  properties {attenuationSettings=&/Spot_HollowBase_Fan_Blades_Att}
 }
 {
  id=Attenuation_1
  children [WavePlayer_3]
  properties {attenuationSettings=&/Spot_HollowBase_Fan_Air_Att}
 }
 {
  id=Enveloper_0
  children [Mixer_0]
  properties {pitchCurve {editorCurveData {keys [{time=0}]}}, volumeCurve {editorCurveData {keys [<at 0 0>,<at 1 1>]}}}
 }
 {id=Mixer_0, children [Attenuation_0, Attenuation_1], properties {inputVolume [1,2.5]}}
 {id=WavePlayer_3, properties {bLooping=true, soundWaveAssetPtr=&/Spot_HollowBase_Fan_Airy_Loop}}
 {id=WavePlayer_4, properties {bLooping=true, soundWaveAssetPtr=&/Spot_HollowBase_Fan_Blades_Loop}}
]
properties {
 bOverrideConcurrency=true
 maxDistance=3500
 soundClassObject=&Fort_Ambient_Props_Athena
 volumeMultiplier=0.35
 concurrencyOverrides {maxCount=2, voiceStealReleaseTime=1}
 concurrencySet [/Game/Sounds/Concurrency_Rules/Athena_Prop]
}
