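# Ambient spot SoundCue for running toilet cistern water at the Hollow Base POI, routed to the Fort_Ambient_OneShots sound class.
# A looping mono wave player feeds an Enveloper that ramps volume from silence to about half over two seconds, offset by a boosted cue volume.
# Overridden attenuation gives short-range logarithmic falloff with focus-based volume scaling.
# Overridden concurrency caps it at two voices with a timed steal release.
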SoundCue /Game/Sounds/Artemis/POI/Hollow_Base/Spots/Spot_Toilet_Cistern_Water_Running_Cue
root=Enveloper_0
nodes [
 {
  id=Enveloper_0
  children [WavePlayer_5]
  properties {pitchCurve {editorCurveData {keys [{time=0}]}}, volumeCurve {editorCurveData {keys [<at 0 0>,<at 2 0.5132>]}}}
 }
 {id=WavePlayer_5, properties {bLooping=true, soundWaveAssetPtr=&/Spot_Toilet_Cistern_Water_Mono}}
]
properties {
 bOverrideAttenuation=true
 bOverrideConcurrency=true
 maxDistance=562
 soundClassObject=&Fort_Ambient_OneShots
 volumeMultiplier=3.007
 attenuationOverrides {
  distanceAlgorithm=Logarithmic
  falloffDistance=512
  focusDistanceScale=0.9
  focusVolumeAttenuation=1.2
  nonFocusVolumeAttenuation=0.4
  stereoSpread=700
  attenuationShapeExtents=<50 64.5942 90>
  customAttenuationCurve {editorCurveData {keys [{arriveTangent=-2.3812, interpMode=RCIM_Cubic, leaveTangent=-2.3812, tangentMode=RCTM_User, time=0, value=1}, <at 1 0.01 RCIM_Cubic>]}}
 }
 concurrencyOverrides {maxCount=2, voiceStealReleaseTime=2}
}
