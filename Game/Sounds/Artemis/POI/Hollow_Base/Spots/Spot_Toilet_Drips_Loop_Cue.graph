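# Looping ambient spot Sound Cue that plays irregular toilet drips for the Hollow Base POI.
# A Looping node repeats a randomized Delay, which feeds a Modulator varying pitch and volume over an equally weighted Random pick of eighteen stereo drip waves.
# It uses the Ambient OneShots sound class and a shared positional tile-ambient attenuation asset, overrides concurrency to limit simultaneous voices, and disables bus and submix sends.
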SoundCue /Game/Sounds/Artemis/POI/Hollow_Base/Spots/Spot_Toilet_Drips_Loop_Cue
root=Looping_0
nodes [
 {id=Delay_0, children [Modulator_0], properties {delayMax=4, delayMin=1.25}}
 {id=Looping_0, children [Delay_0]}
 {id=Modulator_0, children [Random_0], properties {pitchMax=1.1, pitchMin=1, volumeMin=0.4}}
 {
  id=Random_0
  children [
   WavePlayer_3
   WavePlayer_4
   WavePlayer_5
   WavePlayer_6
   WavePlayer_7
   WavePlayer_11
   WavePlayer_12
   WavePlayer_13
   WavePlayer_14
   WavePlayer_15
   WavePlayer_16
   WavePlayer_17
   WavePlayer_18
   WavePlayer_25
   WavePlayer_26
   WavePlayer_27
   WavePlayer_28
   WavePlayer_29
  ]
  properties {weights [1,1,1,1,1,1,1,1,1,1,1,1,1,1,1,1,1,1]}
 }
 {id=WavePlayer_11, properties {soundWaveAssetPtr=&/Spot_Toilet_Drips_06_Stereo}}
 {id=WavePlayer_12, properties {soundWaveAssetPtr=&/Spot_Toilet_Drips_07_Stereo}}
 {id=WavePlayer_13, properties {soundWaveAssetPtr=&/Spot_Toilet_Drips_08_Stereo}}
 {id=WavePlayer_14, properties {soundWaveAssetPtr=&/Spot_Toilet_Drips_09_Stereo}}
 {id=WavePlayer_15, properties {soundWaveAssetPtr=&/Spot_Toilet_Drips_10_Stereo}}
 {id=WavePlayer_16, properties {soundWaveAssetPtr=&/Spot_Toilet_Drips_11_Stereo}}
 {id=WavePlayer_17, properties {soundWaveAssetPtr=&/Spot_Toilet_Drips_12_Stereo}}
 {id=WavePlayer_18, properties {soundWaveAssetPtr=&/Spot_Toilet_Drips_13_Stereo}}
 {id=WavePlayer_25, properties {soundWaveAssetPtr=&/Spot_Toilet_Drips_14_Stereo}}
 {id=WavePlayer_26, properties {soundWaveAssetPtr=&/Spot_Toilet_Drips_15_Stereo}}
 {id=WavePlayer_27, properties {soundWaveAssetPtr=&/Spot_Toilet_Drips_16_Stereo}}
 {id=WavePlayer_28, properties {soundWaveAssetPtr=&/Spot_Toilet_Drips_17_Stereo}}
 {id=WavePlayer_29, properties {soundWaveAssetPtr=&/Spot_Toilet_Drips_18_Stereo}}
 {id=WavePlayer_3, properties {soundWaveAssetPtr=&/Spot_Toilet_Drips_01_Stereo}}
 {id=WavePlayer_4, properties {soundWaveAssetPtr=&/Spot_Toilet_Drips_02_Stereo}}
 {id=WavePlayer_5, properties {soundWaveAssetPtr=&/Spot_Toilet_Drips_03_Stereo}}
 {id=WavePlayer_6, properties {soundWaveAssetPtr=&/Spot_Toilet_Drips_04_Stereo}}
 {id=WavePlayer_7, properties {soundWaveAssetPtr=&/Spot_Toilet_Drips_05_Stereo}}
]
properties {
 attenuationSettings=/Game/Sounds/Attenuation/Fort_TileAmbient_OneShot_MorePositional_Att
 bEnableBusSends=false
 bEnableSubmixSends=false
 bOverrideConcurrency=true
 maxDistance=8000
 soundClassObject=&Fort_Ambient_OneShots
 volumeMultiplier=0.501
 attenuationOverrides {falloffDistance=6000, stereoSpread=10000, attenuationShapeExtents {x=2000}}
 concurrencyOverrides {maxCount=2, voiceStealReleaseTime=2}
}
