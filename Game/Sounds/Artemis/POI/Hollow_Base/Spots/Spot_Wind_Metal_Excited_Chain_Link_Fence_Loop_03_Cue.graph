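# Looping ambient spot cue of wind rattling a chain-link fence at the Hollow_Base POI.
# One looping wave feeds two nested, endlessly looping 40-second Envelopers: the inner shapes pitch, the outer swells and dips volume to vary the gusts.
# The first two quality-level inputs share this chain, and the other two are empty.
# It plays quietly with interior window-wind attenuation, a two-voice concurrency cap and the Fort_Doors sound class.
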SoundCue /Game/Sounds/Artemis/POI/Hollow_Base/Spots/Spot_Wind_Metal_Excited_Chain_Link_Fence_Loop_03_Cue
root=QualityLevel_0
nodes [
 {
  id=Enveloper_0
  children [WavePlayer_0]
  properties {
   bLoop=true
   bLoopIndefinitely=true
   loopEnd=40
   pitchCurve {editorCurveData {keys [
      {interpMode=RCIM_Cubic, time=0}
      <at 10 0.9 RCIM_Cubic>
      {arriveTangent=-0.0001, interpMode=RCIM_Cubic, leaveTangent=-0.0001, tangentMode=RCTM_User, time=20}
      <at 30 0.9 RCIM_Cubic>
      {interpMode=RCIM_Cubic, time=40}
     ]}}
   volumeCurve {editorCurveData {keys [<at 0 1>]}}
  }
 }
 {
  id=Enveloper_3
  children [Enveloper_0]
  properties {
   bLoop=true
   bLoopIndefinitely=true
   loopEnd=40
   pitchCurve {editorCurveData {keys [{time=0}]}}
   volumeCurve {editorCurveData {keys [
      {arriveTangent=-0.01, interpMode=RCIM_Cubic, time=0, value=1}
      {arriveTangent=0.0022, interpMode=RCIM_Cubic, leaveTangent=0.0022, tangentMode=RCTM_User, time=7.8002, value=0.5601}
      {arriveTangent=0.0044, interpMode=RCIM_Cubic, leaveTangent=0.0044, tangentMode=RCTM_User, time=16.1272, value=0}
      {arriveTangent=-0.0003, interpMode=RCIM_Cubic, leaveTangent=-0.0003, tangentMode=RCTM_User, time=20, value=1}
      {arriveTangent=-0.1039, interpMode=RCIM_Cubic, leaveTangent=-0.1039, tangentMode=RCTM_User, time=27.7045, value=0.782}
      {arriveTangent=0.0016, interpMode=RCIM_Cubic, leaveTangent=0.0016, tangentMode=RCTM_User, time=29.9314, value=0.1}
      {arriveTangent=0.0894, interpMode=RCIM_Cubic, leaveTangent=0.0894, time=34.8458, value=0.2616}
      <at 40 1>
     ]}}
  }
 }
 {id=QualityLevel_0, children [Enveloper_3,Enveloper_3,null,null]}
 {
  id=WavePlayer_0
  properties {bLooping=true, soundWaveAssetPtr=&/Spot_Wind_Excited_Chainlink_Fence_Rattle_01}
 }
]
properties {
 attenuationSettings=&/Spot_Interior_Window_Wind_Att
 bOverrideConcurrency=true
 maxDistance=1253.7231
 soundClassObject=&Fort_Doors
 volumeMultiplier=0.3
 attenuationOverrides {falloffDistance=1600}
 concurrencyOverrides {maxCount=2, voiceStealReleaseTime=2}
}
