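# Looping spot ambience cue for the Hollow Base POI, playing wind-excited vibrating metal and nylon.
# Two nested, indefinitely looping Envelopers of different lengths add slow pitch and volume drift to one looping wave.
# The quality node routes its first two levels to that chain and leaves the other two empty.
# Overrides set capsule attenuation with HRTF, distance low-pass and listener focus, plus a single-voice limit.
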SoundCue /Game/Sounds/Artemis/POI/Hollow_Base/Spots/Spot_Wind_Metal_Excited_Vibrating_Nylon_Loop_Cue
root=QualityLevel_0
nodes [
 {
  id=Enveloper_0
  children [WavePlayer_1]
  properties {
   bLoop=true
   bLoopIndefinitely=true
   loopEnd=30
   pitchCurve {editorCurveData {keys [{interpMode=RCIM_Cubic, time=0}, <at 15.9468 0.9181 RCIM_Cubic>, {interpMode=RCIM_Cubic, time=30}]}}
   volumeCurve {editorCurveData {keys [<at 0 1>]}}
  }
 }
 {
  id=Enveloper_3
  children [Enveloper_0]
  properties {
   bLoop=true
   bLoopIndefinitely=true
   loopEnd=20
   pitchCurve {editorCurveData {keys [{time=0}]}}
   volumeCurve {editorCurveData {keys [
      {arriveTangent=-0.01, interpMode=RCIM_Cubic, time=0, value=1}
      <at 10 1.2 RCIM_Cubic>
      {interpMode=RCIM_Cubic, leaveTangent=0.01, time=20, value=1}
     ]}}
  }
 }
 {id=QualityLevel_0, children [Enveloper_3,Enveloper_3,null,null]}
 {
  id=WavePlayer_1
  properties {bLooping=true, soundWaveAssetPtr=&/Spot_Wind_Excited_Metal_Vibrating_Nylon_Loop}
 }
]
properties {
 bOverrideAttenuation=true
 bOverrideConcurrency=true
 maxDistance=1000
 soundClassObject=&Fort_Gadgets
 attenuationOverrides {
  attenuationShape=Capsule
  bAttenuateWithLPF=true
  bEnableListenerFocus=true
  binauralRadius=400
  falloffDistance=750
  focusVolumeAttenuation=1.1
  lPFFrequencyAtMax=1000
  lPFRadiusMax=2000
  lPFRadiusMin=512
  nonFocusVolumeAttenuation=0.75
  spatializationAlgorithm=SPATIALIZATION_HRTF
  attenuationShapeExtents=<250 150>
 }
 concurrencyOverrides {maxCount=1, voiceStealReleaseTime=3}
}
